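; Athena supply drop llama, a loot container that spawns pickups from a loot table when searched or destroyed.
; The server binds to the death of the building actor beneath it and multicasts gravity so it falls when support disappears.
; Opening sets Looted, whose rep-notify plays FX and hides it, and updates quests outside Creative.
; Clients show a repeating chest sound indicator and snap to the replicated final destination.
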(event Custom|Show%20audio%20indicator
  (bind _sound_ambient (Variables|Default|GetSound_Ambient))
  ;[1 1 1] Audio visual indicator
  (Utilities|IsValid _sound_ambient
    (:Is%20Valid
      (Fortnite|ShowSoundIndicatorLocal :Location (Transformation|GetActorLocation) :MaxAudibleDistance (Variables|/Script/Engine.SoundBase|GetMaxDistance _sound_ambient) :IndicatorType "Chest" :Instigator self :Tint "(R=0.000000,G=0.000000,B=0.000000,A=0.000000)" :AttenuationMaxRangeMultiplier 1.0 :AttenuationIntensityMultiplier 1.0))))

(event EventBeginPlay
  (bind (_bblockinghit _binitialoverlap _time _distance _location _impactpoint _normal _impactnormal _physmat _hitactor) (Collision|BreakHitResult (GroundCollision|GroundTrace (Transformation|GetActorLocation) (+ (Transformation|GetActorLocation) "0, 0,-200.000000"))))
  ;[1 1 1] BeginPlay - Bind event for when the building actor under us has died
  (Networking|SwitchHasAuthority
    (:Authority
      (Utilities|FlowControl|Sequence
        (:then_0
          (CallFunction|SetupBindtoBuildingActor _hitactor)
          (Variables|Default|SetFinalDestination (Transformation|GetActorLocation))
          ;[1 1 1] SERVER - Check if In Creative Mode or not
          (bind _asfort_game_state_athena (Utilities|Casting|CastToFortGameStateAthena (Game|GetGameState)))
          (if (Variables|/Script/FortniteGame.FortGameStateAthena|GetbPlaylistDataIsLoaded _asfort_game_state_athena)
            (if (GameplayTags|HasTag (Class|FortGameStateAthena|GetAthenaPlaylistContextTags _asfort_game_state_athena) "(TagName=\"Athena.Creative\")")
              (Variables|Default|SetCreativeMode true)
              (else
                (Variables|Default|SetCreativeMode false)))
            (else
              (Playlist|BindEventtoOnCurrentPlaylistLoaded _asfort_game_state_athena (delegate Custom|PlaylistLoaded)))))))
    (:Remote
      ;[1 1 1] Audio visual indicator
      (bind _returnvalue (Utilities|Time|SetTimerbyEvent (delegate Custom|Show%20audio%20indicator) (Variables|Default|GetSoundIndicatorUpdateInterval) true))
      (Variables|Default|SetSoundIndicatorTimerHandle _returnvalue))))

(event Custom|PlaylistLoaded
  (params (PlaylistName :type Name) (PlaylistContextTags :type Gameplay%20Tag%20Container%20Structure%20%28by%20ref%29))
  ;[1 1 1] SERVER - Check if In Creative Mode or not
  (if (GameplayTags|HasTag PlaylistContextTags "(TagName=\"Athena.Creative\")")
    (Variables|Default|SetCreativeMode true)
    (else
      (Variables|Default|SetCreativeMode false))))

(event OnProjectileStop(ProjectileMovement)
  (params (ImpactResult :type Hit%20Result%20Structure%20%28by%20ref%29))
  (bind (_bblockinghit _binitialoverlap _time _distance _location _impactpoint _normal _impactnormal _physmat _hitactor) (Collision|BreakHitResult ImpactResult))
  ;[1 1 1] If the llama falls and lands on a building actor again, bind to fall if it goes away
  (Networking|SwitchHasAuthority
    (:Authority
      (CallFunction|SetupBindtoBuildingActor _hitactor))))

(event Custom|OnLandingLocationChanged
  (params (DamagedActor :type Actor%20Object%20Reference) (Damage :type Float%20%28single-precision%29) (InstigatedBy :type Controller%20Object%20Reference) (DamageCauser :type Actor%20Object%20Reference) (HitLocation :type Vector) (FHitComponent :type Primitive%20Component%20Object%20Reference) (BoneName :type Name) (Momentum :type Vector))
  ;[1 1 1] Event fired when the ground this supply drop was set to land on has died
  (CallFunction|MulticastApplyGravityforFall))

(event Custom|Multicast_ApplyGravityForFall
  (bind _projectilemovement (Variables|Default|GetProjectileMovement))
  ;[1 1 1] Applies the Gravity client & server so the llama falls when its support structure goes away
  (Components|Movement|SetUpdatedComponent _projectilemovement (Variables|Transformation|GetRootComponent))
  (Variables|/Script/Engine.ProjectileMovementComponent|SetProjectileGravityScale 1.0 _projectilemovement)
  (Networking|FlushNetDormancy))

(event EventOnDeathServer
  (params (Damage :type Float%20%28single-precision%29) (DamageTags :type Gameplay%20Tag%20Container%20Structure%20%28by%20ref%29) (Momentum :type Vector :default "0, 0, 0") (HitInfo :type Hit%20Result%20Structure%20%28by%20ref%29) (InstigatedBy :type Controller%20Object%20Reference) (DamageCauser :type Actor%20Object%20Reference) (EffectContext :type Gameplay%20Effect%20Context%20Handle%20Structure))
  (bind _objbackendname (Variables|QuestStuff|GetObjBackendName))
  (bind _controllerwhoopenedllama (Variables|QuestStuff|GetControllerWhoOpenedLlama))
  (bind _questitem (Variables|QuestStuff|GetQuestItem))
  ;[1 1 1] Server Interact / Death
  (bind _asfort_player_controller (Utilities|Casting|CastToFortPlayerController InstigatedBy)
    (:then
      (Variables|QuestStuff|SetControllerWhoOpenedLlama _asfort_player_controller)
      (if (not (Variables|Looted|GetLooted))
        (CallFunction|MulticastPlayOpenSoundonDeath)))
    (:CastFailed
      _))
  (if (not (Variables|Looted|GetLooted))
    (Variables|Looted|SetLooted true)
    (CallFunction|SpawnLoot)
    (CallFunction|SpawningLootonDestruction InstigatedBy)
    (Actor|SetLifeSpan 2.0)
    ;[1 1 1] Quest Hooks for opening a Llama if not creative mode
    (if (not (Variables|Default|GetCreativeMode))
      (bind (_questisvalid _questcompleted) (Default|AthenaCheckQuestandObjectiveCompletion _controllerwhoopenedllama _questitem _objbackendname))
      (if (and _questisvalid (not _questcompleted))
        (Class|AthenaQuestFunctionLibrary|AthenaGiveQuestUpdatetoPlayers :InQuestItemRef _questitem :ObjectiveBackendName _objbackendname :PlayerControllersForUpdate (Utilities|Array|MakeArray _controllerwhoopenedllama))))))

(event Interaction|EventBlueprintOnInteract
  (params (InteractingPawn :type Fort%20Pawn%20Object%20Reference) (InteractionBeingAttempted :type EInteractionBeingAttempted%20Enum :default "FirstInteraction"))
  (bind _returnvalue (Pawn|GetController InteractingPawn))
  (bind _objbackendname (Variables|QuestStuff|GetObjBackendName))
  (bind _controllerwhoopenedllama (Variables|QuestStuff|GetControllerWhoOpenedLlama))
  (bind _questitem (Variables|QuestStuff|GetQuestItem))
  ;[1 1 1] Server Interact / Death
  (bind _asfort_player_controller (Utilities|Casting|CastToFortPlayerController _returnvalue)
    (:then
      (bind _output_get (Variables|QuestStuff|SetControllerWhoOpenedLlama _asfort_player_controller))
      (CallFunction|MulticastPlayOpenSoundonLooted _output_get))
    (:CastFailed
      (Utilities|Casting|CastToFortAthenaAIBotController _returnvalue)))
  (if (not (Variables|Looted|GetLooted))
    (Variables|Looted|SetLooted true)
    (CallFunction|SpawnLoot)
    (CallFunction|SpawningLootonDestruction (stale EventOnDeathServer InstigatedBy))
    (Actor|SetLifeSpan 2.0)
    ;[1 1 1] Quest Hooks for opening a Llama if not creative mode
    (if (not (Variables|Default|GetCreativeMode))
      (bind (_questisvalid _questcompleted) (Default|AthenaCheckQuestandObjectiveCompletion _controllerwhoopenedllama _questitem _objbackendname))
      (if (and _questisvalid (not _questcompleted))
        (Class|AthenaQuestFunctionLibrary|AthenaGiveQuestUpdatetoPlayers :InQuestItemRef _questitem :ObjectiveBackendName _objbackendname :PlayerControllersForUpdate (Utilities|Array|MakeArray _controllerwhoopenedllama))))))

(event Custom|Multicast_PlayOpenSoundOnLooted
  (params (ControllerWhoOpenedLlama :type Fort%20Player%20Controller%20Object%20Reference))
  ; Multicast Play Open Sound on Looted
  (Networking|SwitchHasAuthority
    (:Remote
      (if (not (Pawn|IsLocallyViewed (Game|Player|GetPlayerPawn ControllerWhoOpenedLlama)))
        (SoundLibrary|PlayEvent (Variables|Default|GetSoundLibrary) "(TagName=\"SoundLibrary.Item.Interact.Complete\")")))))

(event Custom|Multicast_PlayOpenSoundOnDeath
  ; Multicast Play Open Sound on Death
  (Networking|SwitchHasAuthority
    (:Remote
      (SoundLibrary|PlayEvent (Variables|Default|GetSoundLibrary) "(TagName=\"SoundLibrary.Item.Interact.Complete\")"))))

(fn Analytics
  (params (ItemsToSpawn :type Array%20of%20Fort%20Item%20Entry%20Structures))
  (Utilities|IsValid (Variables|QuestStuff|GetControllerWhoOpenedLlama)
    (:Is%20Valid
      (Analytics|SendSpawnLootAnalytics (Variables|QuestStuff|GetControllerWhoOpenedLlama) (Variables|GetItemsToSpawn)))))

(fn BlueprintCanInteract
  (params (InteractingPawn :type Fort%20Pawn%20Object%20Reference) (InteractionBeingAttempted :type EInteractionBeingAttempted%20Enum :default "FirstInteraction") (InteractionType :type TInteractionType%20Enum :default "IT_NoInteraction"))
  (return (not (Variables|Looted|GetLooted))))

(fn BlueprintGetInteractionString
  (params (InteractingPawn :type Fort%20Pawn%20Object%20Reference) (InteractionBeingAttempted :type EInteractionBeingAttempted%20Enum :default "FirstInteraction"))
  (return "NSLOCTEXT(\"[32F3BDCF4E98B269FE6F6994042FB311]\", \"8640A677424EF590D6E63B8DC329CAD7\", \"Search\")"))

(fn GetLootSpawnLocation
  (return (Math|Vector|vector+vector (Transformation|GetWorldLocation (Variables|Default|GetLlamaStaticMesh)) (* (Math|Random|RandomUnitVector) (Math|Random|RandomFloatinRange 20 45)) (Variables|Default|GetLootSpawnOffset))))

(fn OnRep_FinalDestination
  (Networking|SwitchHasAuthority
    (:Remote
      (Game|Components|ProjectileMovement|StopSimulating (Variables|Default|GetProjectileMovement) (LocalVariables|GetFakeHitResult))
      (Transformation|SetActorLocation (Variables|Default|GetFinalDestination) false true))))

(fn OnRep_Looted
  (Collision|SetActorEnableCollision false)
  (CallFunction|PlayLootedFX)
  (Utilities|Time|ClearandInvalidateTimerbyHandle (Variables|Default|GetSoundIndicatorTimerHandle)))

(fn OnSetCustomDepthStencilValue
  (params (PrimComponents :type Array%20of%20Primitive%20Component%20Object%20References) (bUseCustomDepth :type Boolean) (StencilValue :type Integer))
  (for _array_element PrimComponents
    (Rendering|SetCustomDepthStencilValue _array_element (Variables|GetStencilValue))
    (if (Variables|GetbUseCustomDepth)
      (Rendering|SetRenderCustomDepth _array_element true)))
  (return true))

(fn PlayLootedFX
  (bind _fxsocketname (Variables|Default|GetFXSocketName))
  (bind _llamastaticmesh (Variables|Default|GetLlamaStaticMesh))
  (bind _returnvalue (Transformation|GetSocketTransform _llamastaticmesh _fxsocketname "RTS_World"))
  (bind _location (.location _returnvalue))
  (bind _rotation (.rotation _returnvalue))
  (Networking|CanExecuteCosmeticEvents
    (:True
      (Utilities|FlowControl|DoOnce false
        (:Completed
          (Collision|SetActorEnableCollision false)
          (if (Variables|Default|GetIsGoldenLlama)
            (Niagara|SpawnSystematLocation (Variables|Default|GetGoldenLlamaFX) _location _rotation)
            (else
              (Effects|SpawnEmitteratLocation (Variables|Looted|GetLlamaLootFX) _location _rotation)))
          (Audio|Components|Audio|FadeOut (Variables|Default|Gettreasure_pinata_loop_Cue) 0.25 0.0 "Linear" :!class /Script/Engine.AudioComponent)
          (Audio|SpawnSoundAttached (Variables|Looted|GetSound_Looted) (Variables|Default|GetLlamaStaticMesh) :!pinned true) ; Play sound for body sequence
          (Rendering|SetActorHiddenInGame true))))))

(fn SetupBindToBuildingActor
  (params (Object :type Object%20Reference))
  (bind _asbuilding_actor (Utilities|Casting|CastToBuildingActor Object))
  (EventDispatchers|BindEventtoOnDied _asbuilding_actor (EventDispatchers|CreateEvent :!event OnLandingLocationChanged) :!class /Script/FortniteGame.BuildingActor))

(fn SpawnLoot
  (Networking|SwitchHasAuthority
    (:Authority
      (Utilities|FlowControl|DoOnce false
        (:Completed
          (bind _returnvalue (Loot|GetLootTierGroupOverride))
          (bind (_outloottodrop _returnvalue_2) (Loot|PickLootDrops (select (Utilities|Name|Equal(Name) _returnvalue "None") (Variables|Default|GetLootTableName) _returnvalue) (Loot|GetLootLevel) -1 "(GameplayTags=)"))
          (if _returnvalue_2
            (CallFunction|Analytics _outloottodrop)
            (for _array_element _outloottodrop
              (bind _asfort_world_item_definition (Utilities|Casting|CastToFortWorldItemDefinition (Class|FortItemFunctionLibrary|BreakFortItemEntry _array_element)))
              (CallFunction|SpawnPickup _asfort_world_item_definition (.Count (Class|FortItemFunctionLibrary|BreakFortItemEntry _array_element)) (Variables|Default|GetMyInteractingPawn) (CallFunction|GetLootSpawnLocation)))))))))

(fn ConstructionScript)
